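; Parent gameplay ability that regenerates the player pawn's shield after a delay without damage, when its data-driven enable row is on.
; Damage received or shielded events remove the active effect and restart a retriggerable delay, which then applies the regen effect and plays a cue.
; A SetByCaller magnitude scales regen by max shield and effect period for smooth ticks.
; A looping timer removes the effect at full shield.
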(event Custom|ApplyShieldRegenAfterDelay
  ;[1 1 1] Apply Effect - After Retrigger Delay Apply Effect
  (if (not (Math|Float|NearlyEqual(Float) (FortPawn|GetShield (Variables|ParentShieldRegen|GetPlayerPawn)) (FortPawn|GetMaxShield (Variables|ParentShieldRegen|GetPlayerPawn))))
    (Utilities|FlowControl|RetriggerableDelay (ScalableFloat|GetValueatLevel (Variables|ParentShieldRegen|GetRow_ShieldRegenDelay)))
    (Development|PrintString "ShieldRegen: Regen Start")
    (bind _returnvalue (CallFunction|MakeEffect))
    (bind _returnvalue_2 (Ability|ApplyGameplayEffectSpecToOwner _returnvalue))
    (Variables|ParentShieldRegen|SetGEH_ShieldRegen _returnvalue_2)
    (Ability|ExecuteGameplayCueOnOwner (Variables|Default|GetT_PotConsumed))
    ;[1 1 1] Check to see if we have full shield, If so remove the effect
    (bind _returnvalue_3 (Utilities|Time|SetTimerbyEvent (delegate Custom|Shield%20Check) (Variables|ParentShieldRegen|GetTime_FullShieldCheck) true))
    (Variables|ParentShieldRegen|SetTimer_FullShieldCheck _returnvalue_3)))

(event Custom|Shield%20Check
  ;[1 1 1] Check to see if we have full shield, If so remove the effect
  (if (Math|Float|NearlyEqual(Float) (FortPawn|GetShield (Variables|ParentShieldRegen|GetPlayerPawn)) (FortPawn|GetMaxShield (Variables|ParentShieldRegen|GetPlayerPawn)))
    (Ability|RemoveGameplayEffectFromOwnerWithHandle (Variables|ParentShieldRegen|GetGEH_ShieldRegen))
    (Utilities|Time|ClearandInvalidateTimerbyHandle (Variables|ParentShieldRegen|GetTimer_FullShieldCheck))))

(event Ability|EventActivateAbility
  ;[1 1 1] Begin Play - Save Out Pawn, Check If enabled
  (if (CallFunction|RowtoBool)
    (bind _asfort_player_pawn (Utilities|Casting|CastToFortPlayerPawn (Ability|GetAvatarActorfromActorInfo)))
    (Variables|ParentShieldRegen|SetPlayerPawn _asfort_player_pawn)
    ;[1 1 1] On Ability Applied wait for damage taken and then start the regen delay
    (Development|PrintString "ShieldRegen: Ability Applied")
    (Ability|Tasks|WaitGameplayEvent "(TagName=\"Event.Damage.Received\")"
      (:then
        (Ability|Tasks|WaitGameplayEvent "(TagName=\"Event.Damage.Shielded\")"
          (:then
            (CallFunction|ApplyShieldRegenAfterDelay))
          (:EventReceived
            _)))
      (:EventReceived
        _))
    ;[1 1 1] On Damage Taken Remove effect and Retrigger the wait
    (Development|PrintString "ShieldRegen: Damage Taken")
    (Ability|RemoveGameplayEffectFromOwnerWithHandle (Variables|ParentShieldRegen|GetGEH_ShieldRegen))
    (CallFunction|ApplyShieldRegenAfterDelay)))

(event Ability|EventOnEndAbility
  (params (bWasCancelled :type Boolean))
  ;[1 1 1] End Ability - Remove Effect
  (Ability|RemoveGameplayEffectFromOwnerWithHandle (Variables|ParentShieldRegen|GetGEH_ShieldRegen)))

(fn MakeEffect
  ;[1 1 1] Make Spec. NOTE that the percent per second is mult by the players max health to get how much health per second which is then mult by .1 as the period on the GE is .1s not 1. as to make sure that the shield regen looks smooth.
  (bind _returnvalue (Ability|GameplayEffect|AssignTagSetbyCallerMagnitude (Ability|MakeOutgoingGameplayEffectSpec (Variables|ParentShieldRegen|GetGE_ShieldRegen)) "(TagName=\"SetByCaller.ShieldRegen\")" (* (* (ScalableFloat|GetValueatLevel (Variables|ParentShieldRegen|GetRow_RegenPercentOfMaxPerSecond)) (FortPawn|GetMaxShield (Variables|ParentShieldRegen|GetPlayerPawn))) (ScalableFloat|GetValueatLevel (Variables|ParentShieldRegen|GetRow_PeriodScaler)))))
  (return _returnvalue))

(fn RowToBool
  (return (Math|Conversions|ToBoolean(Integer) (Math|Float|Ceil (ScalableFloat|GetValueatLevel (Variables|ParentShieldRegen|GetRow_Enabled))))))
